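# Translucent UI material for the player arrow on the Athena radar, with an optional pulsing flash.
# Two anti-aliased thresholds on a grayscale distance-field icon, scaled about its center, form the arrow's fill and dark outline.
# When bIsFlashActive is above zero, two CirclePulse function rings, tinted by PulseColor, are combined with the arrow.
# Unconnected BrushScale and TotalFlashTimeInSeconds parameters exist for code to read and must not be deleted.
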Material /Game/UI/CodeReferenced/UI_Radar_AthenaPlayerArrow_Mat
graph {
 properties {
  bAutomaticallySetUsageInEditor=false
  bUsedWithStaticMesh=false
  blendMode=BLEND_Translucent
  materialDomain=MD_UI
  pixelDepthOffsetMode=PDOM_Legacy
  usedShadingModels=Default Lit
 }
 nodes [
  DoubleVectorParameter() {
   desc="This value is read in by code; don't delete\r\n\r\nThis scales both the icon and pulse effect; modify IconScale accordingly"
   parameterName=BrushScale
   defaultValue=<1 1 0 1>
  }
  n1 = ScalarParameter() {parameterName=bIsFlashActive}
  n2 = VectorParameter() {parameterName=PulseColor, defaultValue=<rgba 1 1 1 1>}
  n3 = MaterialFunctionCall(Size=Constant() {r=0.05}, GlowMax=Constant() {r=0.5}) {materialFunction=&MF_UI_Circle}
  n4 = TextureSampleParameter2D(UVs=MaterialFunctionCall(UVs=TextureCoordinate(), Texture Scale=ScalarParameter() {defaultValue=0.2, group=Icon, parameterName=IconScale}) {materialFunction=&ScaleUVsByCenter}.UVs) {
   group=Icon
   parameterName=IconTexture
   samplerSource=SSM_Clamp_WorldGroupSettings
   samplerType=SAMPLERTYPE_LinearGrayscale
   texture=&/T_UI_MinimapArrow
  }
  n5 = MaterialFunctionCall(DistanceField=@n4.RGB, Threshold=Constant() {r=0.5}) {materialFunction=&MF_UI_AntiAliasing}
  n6 = LinearInterpolate(A=LinearInterpolate(A=@n2.RGB, B=Add(A=@n2.RGB) {constB=0.1}, Alpha=Multiply(A=@n3.Fill) {constB=0.5}), Alpha=@n5.Outer) {constB=0}
  n7 = MaterialFunctionCall(DistanceField=@n4.RGB, Threshold=Constant() {r=0.85}) {materialFunction=&MF_UI_AntiAliasing}
  n8 = Add(A=@n7.Outer, B=@n7.Outer) {constB=0}
  n9 = OneMinus(None=Saturate(None=@n8))
  If(A=@n1, "A > B"=Add(A=Multiply(A=@n6, B=@n9), B=Min(A=@n8, B=@n7.Outer)), "A == B"=@n7.Outer, "A < B"=@n7.Outer)
  n10 = LinearInterpolate(A=@n7.Outer, B=@n5.Outer) {constAlpha=1}
  n11 = ScalarParameter() {parameterName=Width Start}
  n12 = ScalarParameter() {group=PulseStyle, parameterName="ScaleStart (S)"}
  n13 = Constant()
  n14 = MaterialFunctionCall!(
   PulseSpeed=ScalarParameter() {defaultValue=1, parameterName=Pulse Speed}
   SizeEnd=ScalarParameter() {defaultValue=0.8, group=PulseStyle, parameterName=ScaleEnd}
   WidthStart=@n11
   WidthEnd=ScalarParameter() {defaultValue=0.1, group=PulseStyle, parameterName="WidthEnd (S)"}
   PulseAnimFalloff
   SizeStart=@n12
   OpacityStart
   OpacityEnd=@n13
   TimeSpeed=ScalarParameter() {defaultValue=1, parameterName=Time}
   TexCoord
  ) {materialFunction=&/MF_AthenaPlayerArrow_CirclePulse}
  n15 = MaterialFunctionCall(
   PulseSpeed=ScalarParameter() {defaultValue=1, parameterName=Pulse Speed2}
   SizeEnd=ScalarParameter() {defaultValue=0.4, group=PulseStyle, parameterName=ScaleEnd2}
   WidthStart=@n11
   WidthEnd=ScalarParameter() {defaultValue=0.2, group=PulseStyle, parameterName=WidthEnd2}
   PulseAnimFalloff=ScalarParameter() {defaultValue=1, group=PulseStyle, parameterName="PulseAnimFalloff (S)"}
   SizeStart=@n12
   OpacityEnd=@n13
  ) {materialFunction=&MF_AthenaPlayerArrow_CirclePulse}
  If(A=@n1, "A > B"=Max(A=@n10, B=Max(A=@n14.Result, B=@n15.Result)), "A == B"=@n10, "A < B"=@n10)
  ScalarParameter() {defaultValue=1.5, desc="This value is read in by code; don't delete", parameterName=TotalFlashTimeInSeconds}
 ]
}
